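# Spot ambient cue for DreamFog that plays Queen of Thorns (TangyRadish) villain laughs as positioned outdoor one-shots, despite the Whispers name.
# A quality node sends the first two levels to an equal-weight random pick of three laughs with slight volume variation.
# The last two levels reuse a single laugh with wide pitch and volume modulation for variety.
# It uses wide tile-ambient attenuation and the Athena outdoor ambient sound class.
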SoundCue /DreamFog/Assets/Audio/IP_Spot_DreamFog_TangyRadish_QueenOfThorns_Whispers_Cue
root=QualityLevel_0
nodes [
 {id=Modulator_0, children [WavePlayer_3], properties {pitchMax=1.1, pitchMin=0.7, volumeMax=1.1, volumeMin=0.5}}
 {id=Modulator_2, children [Random_0], properties {pitchMax=1, pitchMin=1, volumeMax=1.1, volumeMin=0.9}}
 {id=QualityLevel_0, children [Modulator_2,Modulator_2,Modulator_0,Modulator_0]}
 {id=Random_0, children [WavePlayer_0,WavePlayer_1,WavePlayer_2], properties {weights [1,1,1]}}
 {id=WavePlayer_0, properties {soundWaveAssetPtr=&/VO_TangyRadish_QueenOfThorns_SpookyVillainLaugh_010}}
 {id=WavePlayer_1, properties {soundWaveAssetPtr=&/VO_TangyRadish_QueenOfThorns_VillainLaugh_010}}
 {id=WavePlayer_2, properties {soundWaveAssetPtr=&/VO_TangyRadish_QueenOfThorns_VillainLaugh_011}}
 {id=WavePlayer_3, properties {soundWaveAssetPtr=&VO_TangyRadish_QueenOfThorns_VillainLaugh_011}}
]
properties {
 attenuationSettings=/Game/Sounds/Attenuation/Fort_TileAmbient_OneShot_Wider_Att
 maxDistance=8000
 soundClassObject=&Fort_Ambient_Outdoors_Athena
 volumeMultiplier=0.9
 preEffectBusSends [{}]
}
