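; A Dynamic Mesh Actor that procedurally builds a looping ring mesh, apparently as a takeover border effect.
; On mesh rebuild, Geometry Script creates a circle path from the component's world transform, Radius and Num Points.
; Its points fill Transform Array as the sweep path, while the same circle, scaled down, serves as the swept profile.
; Transform Array is never cleared here, and the other events are empty.
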(event EventBeginPlay)

(event Collision|EventActorBeginOverlap
  (params (OtherActor :type Actor%20Object%20Reference)))

(event EventTick
  (params (DeltaSeconds :type Float%20%28single-precision%29)))

(event Events|EventOnRebuildGeneratedMesh
  (params (TargetMesh :type Dynamic%20Mesh%20Object%20Reference))
  (bind _dynamicmeshcomponent (Variables|DynamicMeshActor|GetDynamicMeshComponent))
  (bind _returnvalue (Transformation|GetWorldTransform _dynamicmeshcomponent))
  (bind _returnvalue_2 (GeometryScript|PolyPath|CreateCirclePath3D _returnvalue (Variables|Default|GetRadius) (Variables|Default|GetNum%20Points)))
  (bind _vertexarray (GeometryScript|PolyPath|ConvertPolyPathtoArray _returnvalue_2))
  (bind _vertexarray_2 (GeometryScript|PolyPath|ConvertPolyPathtoArrayOfVector2D _returnvalue_2))
  (for _array_element _vertexarray
    (Utilities|Array|Add (Variables|Default|GetTransform%20Array) (Math|Transform|MakeTransform _array_element "0, 0, 0" "0.015000,0.015000,0.025000")))
  (GeometryScript|Primitives|AppendSweepPolyline :TargetMesh TargetMesh :Transform _returnvalue :PolylineVertices _vertexarray_2 :SweepPath (Variables|Default|GetTransform%20Array) :bLoop true :StartScale 0.1 :EndScale 0.1 :RotationAngleDeg 0.0 :MiterLimit 1.0))

(fn ConstructionScript)
