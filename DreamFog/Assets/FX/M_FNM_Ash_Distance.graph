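# Masked, two-sided material for looping ash and ember sprites animated in the shader, with a camera-distance fade.
# Vertex color staggers each sprite's looping age and varies size, direction, flipbook and spin.
# World position offset math combines a Sprite function with age-driven drift plus panning curl noise.
# Sphere masks derive fade over age and near the camera, while a rotating SubUV ember texture is tinted by a glow color.
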Material /DreamFog/Assets/FX/M_FNM_Ash_Distance
graph {
 properties {bAutomaticallySetUsageInEditor=false, blendMode=BLEND_Masked, twoSided=true, usedShadingModels=Default Lit}
 nodes [
  n1 = Multiply(
   A=NamedRerouteUsage()
   B=LinearInterpolate(
    A=ScalarParameter() {defaultValue=25, parameterName=Scale Min}
    B=ScalarParameter() {defaultValue=40, parameterName=Scale Max}
    Alpha=VertexColor().R
   )
  )
  n2 = LinearInterpolate(
   A=ScalarParameter() {defaultValue=0.1, parameterName=Camera Scale Min}
   B=ScalarParameter() {defaultValue=2, parameterName=Camera Scale Max}
   Alpha=NamedRerouteUsage()
  )
  n3 = VertexColor()
  n4 = Time()
  n5 = ScalarParameter() {defaultValue=0.1, parameterName=Time Scale}
  n6 = ScalarParameter() {defaultValue=1, parameterName=Loop Time}
  n7 = Fmod(A=Add(A=@n3.R, B=Multiply(A=@n4, B=@n5)), B=@n6)
  n8 = WorldPosition()
  n9 = Multiply(A=Time(), B=VectorParameter() {defaultValue=<rgba 0 0 75 0>}.RGB)
  n10 = VectorNoise(World Position=Multiply(A=Add(A=@n8.XYZ, B=@n9), B=ScalarParameter() {defaultValue=0.002, parameterName=Noise Scale})) {noiseFunction=VNF_CurlALU}
  n11 = ScalarParameter() {defaultValue=50, parameterName=Noise Amplitude}
  n12 = VectorParameter() {parameterName=Param_1, defaultValue=<rgba 750 0 -2500 0>}
  n13 = VectorParameter() {parameterName=Direction, defaultValue=<rgba 1500 0 -1960 0>}
  n14 = VertexColor()
  n15 = Multiply(A=Add(A=@n3.A, B=@n7), B=Add(A=Multiply(A=@n10, B=@n11), B=LinearInterpolate(A=@n12.RGB, B=@n13.RGB, Alpha=@n14.G)))
  Add(
   A=MaterialFunctionCall!(Center Location=WorldPosition().XYZ, 0-1 UVs=TextureCoordinate(), XY scale=Multiply(A=@n1, B=@n2), OverridePivot, "Normals (Optional)") {materialFunction=&Sprite}.World position offset
   B=@n15
  )
  MaterialFunctionCall!(Texture, UVs, SubImages, Frame=Multiply(A=NamedRerouteUsage(), B=ScalarParameter() {defaultValue=36, parameterName=XY Length})) {materialFunction=/Engine/Functions/Engine_MaterialFunctions03/Texturing/SubUV_Function}
  n16 = ScalarParameter() {defaultValue=6, parameterName=X}
  n17 = ScalarParameter() {defaultValue=6, parameterName=Y}
  n18 = VertexColor()
  n19 = Multiply(A=ScalarParameter() {defaultValue=1}, B=@n18.G)
  n20 = StaticBoolParameter() {defaultValue=true, parameterName="Manual Time Control (SB)"}
  n21 = Add(A=@n18.G, B=Time())
  n22 = ScalarParameter() {defaultValue=10, parameterName=Rotation Rate}
  n23 = Rotator(Coordinate=TextureCoordinate(), Time=Multiply(A=Time(), B=LinearInterpolate(B=@n22, Alpha=@n18.G)))
  Multiply(
   A=VectorParameter() {parameterName=Ash Glow Color, defaultValue=<rgba 50 0 0 0>}.RGB
   B=TextureSample(UVs=MaterialFunctionCall!(U_Images=@n16, V_Images=@n17, SubUV Texture RGB, Time Speed=@n19, Manual Time Control=@n20, Time=@n21, UVs=@n23) {materialFunction=/Game/Effects/Fort_Effects/Materials/_Functions/MF_SubUV}.UVs) {texture=/Game/Effects/Fort_Effects/Textures/Motes/T_Ash_Embers_Glow_SubUV}.R
  )
  NamedRerouteDeclaration(None=Divide(A=@n7, B=@n6)) {name=Normalized Age, nodeColor=<rgba 0.2187 1 0 1>}
  NamedRerouteDeclaration(None=SphereMask(
    A=NamedRerouteUsage()
    B=ScalarParameter() {defaultValue=0.5, parameterName=Curve Midpoint}
    Radius=ScalarParameter() {defaultValue=0.5, parameterName=Curve Size}
    Hardness=ScalarParameter() {parameterName=Curve Hardness}
   )) {name=Fade, nodeColor=<rgba 0 1 0.9766 1>}
  n24 = WorldPosition()
  n25 = NamedRerouteUsage()
  n26 = Add(A=Add(A=Add(A=@n24.XYZ, B=@n25), B=NamedRerouteUsage()), B=NamedRerouteUsage())
  n27 = CameraPositionWS()
  n28 = ScalarParameter() {defaultValue=7500, parameterName=Near Camera Fade}
  n29 = ScalarParameter() {parameterName=Near Camera Fade Radius Soften}
  NamedRerouteDeclaration(None=OneMinus(None=SphereMask(A=@n26, B=@n27, Radius=@n28, Hardness=@n29))) {name=Camera Fade, nodeColor=<rgba 0.7344 0 1 1>}
  NamedRerouteDeclaration(None=@n1) {name=Size, nodeColor=<rgba 0 0.1562 1 1>}
  NamedRerouteDeclaration(None=@n15) {name=Movement, nodeColor=<rgba 0.4766 1 0 1>}
  NamedRerouteDeclaration() {name=Sprite, nodeColor=<rgba 1 0 0.7969 1>}
  VectorParameter() {parameterName=Param_2, defaultValue=<rgba 0.0747 0.0747 0.0747 0>}
 ]
}
