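# Translucent unlit Niagara mesh-particle material for DreamFog that draws a screen-space VHS-style scanline and static overlay, with depth test disabled.
# Two panned scanline textures blend and add to SubUV-animated noise as "Static", while a tiled radial dot grid forms "Pixels".
# A "VHS Effect" dynamic parameter scales the combined reroute signal, which is tinted by particle color and masked by radial and vertical screen-position falloffs.
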Material /DreamFog/Assets/FX/M_FNM_Scanlines_Particle
graph {
 properties {
  bAutomaticallySetUsageInEditor=false
  bDisableDepthTest=true
  bUsedWithNiagaraMeshParticles=true
  blendMode=BLEND_Translucent
  shadingModel=MSM_Unlit
  twoSided=true
  usedShadingModels=Unlit
 }
 nodes [
  n1 = ScreenPosition()
  n2 = Length(None=Subtract(A=ScreenPosition().ViewportUV, B=ScalarParameter() {defaultValue=0.5, parameterName=Param_23})) {is=class MaterialExpressionLength}
  n3 = ScalarParameter() {defaultValue=3, parameterName=Param_25}
  n4 = NamedRerouteUsage()
  n5 = ScalarParameter() {defaultValue=0.75, parameterName=Param_16}
  n6 = LinearInterpolate(
   B=Add(A=NamedRerouteUsage(), B=Multiply(A=@n4, B=@n5))
   Alpha=DynamicParameter() {paramNames [VHS Effect, Random Offset, Param3, Param4]}.VHS Effect
  )
  n7 = ParticleColor()
  Multiply(
   A=Power(Base=ComponentMask(None=@n1.ViewportUV) {g=true}, Exp=ScalarParameter() {defaultValue=3})
   B=Multiply(A=Power(Base=@n2, Exp=@n3), B=Multiply(A=@n6, B=@n7.RGB))
  )
  n8 = Multiply(
   A=ScreenPosition().ViewportUV
   B=Normalize(VectorInput=MaterialFunctionCall() {materialFunction=&ScreenResolution}.Visible Resolution)
  )
  n9 = ScalarParameter() {defaultValue=750, parameterName=Param_4}
  n10 = Subtract(A=Frac(None=Multiply(A=@n8, B=@n9)), B=ScalarParameter() {defaultValue=0.5, parameterName=Param_5})
  n11 = Power(Base=OneMinus(None=Length(None=@n10)), Exp=ScalarParameter() {defaultValue=2, parameterName=Param_3})
  n12 = ScalarParameter() {defaultValue=5, parameterName=Param_24}
  NamedRerouteDeclaration(None=Saturate(None=Multiply(A=@n11, B=@n12))) {name=Pixels, nodeColor=<rgba 0 0.8125 1 1>}
  n13 = ScreenPosition()
  n14 = Multiply(A=Time(), B=ScalarParameter() {defaultValue=0.25, parameterName=Time Scale})
  n15 = Multiply(
   A=ScalarParameter() {defaultValue=0.1, parameterName=Param_18}
   B=TextureSample(UVs=Panner(Coordinate=@n13.ViewportUV, Time=@n14) {speedY=1}) {texture=&/T_ScanLines_DreamFog}.G
  )
  n16 = Multiply(A=Time(), B=ScalarParameter() {defaultValue=0.125, parameterName=Time Scale 2ndary Scaline})
  n17 = Multiply(
   A=TextureSample(UVs=Panner(Coordinate=@n13.ViewportUV, Time=@n16) {speedY=1}) {texture=&/T_DreamFog_ScanlinesOffsetHoriz_A}.G
   B=ScalarParameter() {defaultValue=0.15, parameterName=Param_19}
  )
  n18 = ScalarParameter() {defaultValue=0.5, parameterName=Param_17}
  n19 = ScalarParameter() {defaultValue=1, parameterName=Param_22}
  n20 = Constant() {r=4}
  n21 = ScreenPosition()
  n22 = Normalize(VectorInput=MaterialFunctionCall() {materialFunction=&ScreenResolution}.Visible Resolution)
  n23 = MaterialFunctionCall!(
   U_Images=@n20
   V_Images=@n20
   SubUV Texture RGB
   Time Speed=ScalarParameter() {defaultValue=8, parameterName=Param_21}
   Manual Time Control=StaticBoolParameter() {defaultValue=true, parameterName="Manual Time Control (SB)_1"}
   Time=Time()
   UVs=Multiply(A=Multiply(A=@n21.ViewportUV, B=@n22), B=ScalarParameter() {defaultValue=16, parameterName=Param_20})
  ) {materialFunction=/Game/Effects/Fort_Effects/Materials/_Functions/MF_SubUV}
  n24 = Add(
   A=TextureSample(UVs=@n23.UVs) {texture=/Game/Global/Textures/Default/PointSplatter/BoatFlake}.G
   B=TextureSample(UVs=@n23.UVs) {texture=/Game/Global/Textures/Default/PointSplatter/T_Ui_ColumbusSeries_512}.G
  )
  NamedRerouteDeclaration(None=Add(A=LinearInterpolate(A=@n15, B=@n17, Alpha=@n18), B=Multiply(A=@n19, B=@n24))) {name=Static, nodeColor=<rgba 0 1 0.2266 1>}
 ]
}
