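# M_SmokeyBorder is a translucent, unlit, two-sided material in the FX folder, named for a smokey border effect.
# Its graph holds only a texture coordinate node and a red vector parameter with an above-one value.
# No connections between these nodes or to material outputs appear, so the file shows no smoke or border logic.
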Material /DreamFog/Assets/FX/M_SmokeyBorder
graph {
 properties {
  bAutomaticallySetUsageInEditor=false
  blendMode=BLEND_Translucent
  shadingModel=MSM_Unlit
  translucencyLightingMode=TLM_VolumetricPerVertexDirectional
  twoSided=true
  usedShadingModels=Unlit
 }
 nodes [
  TextureCoordinate()
  VectorParameter() {defaultValue=<rgba 5 0 0 0>}
 ]
}
