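# Niagara system for a haunted point of interest (per its name), filling an area with glowing red-magenta ash sprites.
# A single GPU emitter, Ash, continuously spawns short-lived, randomly sized SubUV sprites within a large owner-scaled sphere.
# Wind, gravity and drag move the sprites as they rotate and scale up, then back to zero.
# A kill volume with a hardcoded, far-from-zero origin culls particles, likely tying the effect to one map location.
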NiagaraSystem /DreamFog/Assets/FX/NS_FNM_HauntedPOI
dependencies {
 usedDataInterfaces [
  &NiagaraDataInterfaceSpriteRendererInfo
  &NiagaraDataInterfaceCurve
  &NiagaraDataInterfaceCollisionQuery
  &NiagaraDataInterfaceLandscape
 ]
 usedDynamicInputs []
 usedModules [&SystemState]
}
emitters {Ash {
  simTarget=GPUComputeSim
  emitterUpdateScript [
   &EmitterState()
   &SpawnRate(SpawnRate=1250, Use Spawn Probability=false)
  ]
  particleSpawnScript [
   &InitializeParticle(
    Color=<rgba 8 0 0.456 1>
    Lifetime=5
    Sprite Size Mode=::NewEnumerator1
    Uniform Sprite Size Max=25
    Uniform Sprite Size Min=12
    UsePositionOffset=false
    Write Material Random=true
   )
   &ShapeLocation(MaskSpawnGroup=false, Override Seed=false, Sphere Radius=3500, UseOwnerScale=true)
   &SubUVAnimation(
    "Sprite Renderer "=DataInterface() {propertyValues {SpriteRenderer=&self:Ash_0.NiagaraSpriteRendererProperties_0}}
    SubUV Animation Mode:ENiagara_SubUVLookupModeV2=::NewEnumerator2
    UseEndFrame=false
    UseStartFrame=false
   )
  ]
  particleUpdateScript [
   &ParticleState()
   &ScaleSpriteSize(Uniform Curve Sprite Scale=DataInterface() {propertyValues {Curve {keys [
        {interpMode=RCIM_Cubic, tangentMode=RCTM_Break, time=0, value=0}
        {interpMode=RCIM_Cubic, tangentMode=RCTM_Break, time=0.5, value=1}
        {interpMode=RCIM_Cubic, tangentMode=RCTM_Break, time=1, value=0}
       ]}}})
   &SpriteRotationRate(Rotation Rate=60)
   &WindForce(Wind Speed=<500 0 0>)
   &GravityForce(Gravity=<0 0 -240>)
   &Drag(Use Linear Drag=true, Use Rotational Drag=true)
   &AerodynamicDrag(Use Linear Drag=true, Use Rotational Drag=true)
   &SolveForcesAndVelocity(Clamp Velocity=false, Limit Acceleration=false)
   &AlignSpriteToMeshOrientation(SetAlignment=true, SetFacing=true)
   &KillParticlesInVolume(Sphere Radius=15500, Volume Origin:Vector3f=<-76431 -84503 567.38>)
  ]
  rendererClasses [&NiagaraSpriteRendererProperties]
  renderers [{rendererClass=&NiagaraSpriteRendererProperties}]
 }}
summary {
 systemName=NS_FNM_HauntedPOI
 emitters [{emitterName=Ash, simTarget=GPUComputeSim, rendererClasses [&NiagaraSpriteRendererProperties]}]
 userVariables []
}
