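# Niagara system for a haunted POI lens effect, which by its name plays on exit, showing a teal-blue tint that fades out.
# Its one enabled emitter burst-spawns a mesh particle, with two empty emitters left disabled.
# A color curve briefly brightens the tint, then fades RGB to zero over the particle's life.
# Dynamic material parameters feed the mesh material a fast-rising, scaled curve and a random value.
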NiagaraSystem /DreamFog/Assets/FX/NS_FNM_HauntedPOI_LensEffect_Exit
dependencies {
 usedDataInterfaces [&NiagaraDataInterfaceColorCurve, &NiagaraDataInterfaceCurve]
 usedDynamicInputs []
 usedModules [&SystemState]
}
emitters {
 Empty {bEnabled=false, rendererClasses [], renderers []}
 Empty001 {bEnabled=false, rendererClasses [], renderers []}
 Minimal {
  emitterUpdateScript [
   &EmitterState(Inactive Response=::NewEnumerator1, Life Cycle Mode=::NewEnumerator1, Loop Behavior=::NewEnumerator1, Loop Duration=6, UseLoopDelay=false)
   &SpawnBurst_Instantaneous(Use Spawn Probability=false, UseLoopCountLimit=false)
  ]
  particleSpawnScript [&InitializeParticle(
    Color=<rgba 0 0.4 0.4 0.05>
    Lifetime=5
    Mesh Scale Mode=::NewEnumerator3
    Mesh Uniform Scale=2
    UsePositionOffset=true
    Write Material Random=true
   )]
  particleUpdateScript [
   &ParticleState(Kill Particles When Lifetime Has Elapsed=false)
   &ScaleColor(
    Linear Color Curve=DataInterface() {propertyValues {
      AlphaCurve {keys [<at 0.7588 1>]}
      BlueCurve {keys [<at 0 3>,<at 0.136 4>,<at 0.9903 0>]}
      GreenCurve {keys [<at 0 1.9577>,<at 0.136 2.6102>,<at 0.9903 0>]}
      RedCurve {keys [<at 0 1.1078>,<at 0.136 1.4771>,<at 0.9903 0>]}
     }}
    Scale Mode=::NewEnumerator2
   )
   &DynamicMaterialParameters(
    Index 0 Param 1=&FloatFromCurve(
     FloatCurve=DataInterface() {propertyValues {Curve {keys [
         <at 0 0 RCIM_Cubic>
         {arriveTangent=-0.1433, interpMode=RCIM_Cubic, leaveTangent=-0.1433, tangentMode=RCTM_User, time=0.0994, value=1.0109}
         {interpMode=RCIM_Cubic, tangentMode=RCTM_Break, time=1, value=1}
        ]}}}
     Scale Curve=12
    )
    Index 0 Param 2=&RandomRangeFloat(Minimum=0.25, Maximum=0.75)
    Module.Index0Float0Write=true
    Module.Index0Float1Write=true
    Module.Index0Float2Write=true
    Module.Index0Float3Write=true
   )
  ]
  rendererClasses [&NiagaraMeshRendererProperties]
  renderers [{rendererClass=&NiagaraMeshRendererProperties, rendererIndex=0}]
 }
}
summary {
 systemName=NS_FNM_HauntedPOI_LensEffect_Exit
 emitters [
  {bEnabled=false, emitterName=Empty, rendererClasses []}
  {bEnabled=false, emitterName=Empty001, rendererClasses []}
  {emitterName=Minimal, rendererClasses [&NiagaraMeshRendererProperties]}
 ]
 userVariables []
}
